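# Translucent unlit Niagara sprite material for the lips kiss effect on the DimeBlanketKnot backpack.
# Subtracts a center-scaled copy of the lips texture from the original, with dynamic parameter Inner Size controlling the cutout.
# An optional flipbook is driven by dynamic SubUV phase; its re-centering uses a fixed offset rather than RowColumns.
# Particle alpha multiplies the mask, and a shading path switch applies depth fade except on forward and mobile.
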Material /BRCosmetics/Accessories/FORT_Backpacks/Backpack_F_MED_DimeBlanketKnot/FX/M_Backpack_DimeBlanketKnot_Lips
graph {
 properties {bUsedWithNiagaraSprites=true, blendMode=BLEND_Translucent, pixelDepthOffsetMode=PDOM_Legacy, shadingModel=MSM_Unlit, usedShadingModels=Unlit}
 nodes [
  Add()
  n1 = ScalarParameter() {defaultValue=2, parameterName=RowColumns}
  n2 = MaterialFunctionCall(
   "Animation  Phase (0-1)"=DynamicParameter() {defaultValue=<rgba 0 1 1 1>, paramNames [Inner Size, "SubUV<0-1>", Param3, Param4]}."SubUV<0-1>"
   Number of Rows=@n1
   Number of Columns=@n1
  ) {materialFunction=&FlipBook}
  n3 = TextureSampleParameter2D(UVs=StaticSwitchParameter(True=@n2.UVs, False=TextureCoordinate()) {parameterName=UseFlipbook}) {parameterName=TextureMain, texture=&/T_DimeBlanketKnot_Lips_Kiss}
  n4 = DynamicParameter() {defaultValue=<rgba 0 1 1 1>, paramNames [Inner Size, "SubUV<0-1>", Param3, Param4]}
  n5 = Constant() {r=0.25}
  n6 = Add(A=Divide(A=@n2.UVs, B=@n4.Inner Size) {constB=4}, B=@n5)
  n7 = Divide(A=@n5, B=@n4.Inner Size) {constB=4, desc=Scale the center offset by the scaling factor}
  n8 = TextureCoordinate()
  n9 = DynamicParameter() {defaultValue=<rgba 0 1 1 1>, paramNames [Inner Size, "SubUV<0-1>", Param3, Param4]}
  n10 = TextureSampleParameter2D(UVs=StaticSwitchParameter(
    True=Subtract(A=@n6, B=@n7) {desc=Re offset to maintain center}
    False=MaterialFunctionCall(UVs=@n8, Texture Scale=@n9.Inner Size) {materialFunction=&ScaleUVsByCenter}.UVs
   ) {parameterName=UseFlipbook}) {
   parameterName=Texture_Subtracter
   samplerSource=SSM_Clamp_WorldGroupSettings
   texture=&T_DimeBlanketKnot_Lips_Kiss
  }
  n11 = Multiply(A=ParticleColor().A, B=Subtract(A=Power(Base=@n3.R) {constExponent=0.75}, B=@n10.R))
  n12 = DepthFade(Opacity=@n11, FadeDistance=ScalarParameter() {defaultValue=5, parameterName=Depth Fade Distance, sliderMax=100})
  ShadingPathSwitch(Default=@n12, Deferred=@n12, Forward=@n11, Mobile=@n11)
 ]
}
