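# Niagara system for the DimeBlanketKnot backpack cosmetic, drawing a glowing heart effect that changes with time of day, front-end context and music-reactive preview.
# AirGlow supplies a sprite glow, Heart a mesh, and the Low_End emitters build heart sprites from index-selected position, size and rotation arrays.
# Hearts emits short-lived, three-renderer flare sprites around socket-reader positions.
# The music preview value gates waveform pulsing and gameplay spawning, while the front end always spawns.
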NiagaraSystem /BRCosmetics/Accessories/FORT_Backpacks/Backpack_F_MED_DimeBlanketKnot/FX/NS_DimeBlanketKnot_Backpack
dependencies {
 usedDataInterfaces [
  &NiagaraDataInterfaceCurve
  &NiagaraDataInterfacePlatformSet
  &NiagaraDataInterfaceArrayFloat3
  &NiagaraDataInterfaceArrayFloat2
  &NiagaraDataInterfaceArrayFloat
  &NiagaraDataInterfaceSocketReader
  &NiagaraDataInterfaceColorCurve
 ]
 usedDynamicInputs [&ColorFromCurve]
 usedModules [
  &SystemState
  &self:SystemSpawnScript.SystemScriptSource.SystemScriptGraph.NiagaraNodeAssignment_2.SetVariables_#1
  &self:SystemSpawnScript.SystemScriptSource.SystemScriptGraph.NiagaraNodeAssignment_3.SetVariables_#2
  &self:SystemSpawnScript.SystemScriptSource.SystemScriptGraph.NiagaraNodeAssignment_0.SetVariables_#3
  &self:SystemSpawnScript.SystemScriptSource.SystemScriptGraph.NiagaraNodeAssignment_1.SetVariables_#4
 ]
}
emitters {
 AirGlow {
  emitterUpdateScript [
   &EmitterState(Life Cycle Mode=::NewEnumerator1, Loop Behavior=::NewEnumerator1, Loop Duration Mode=::NewEnumerator1, UseLoopDelay=false)
   &SpawnBurst_Instantaneous(Spawn Count=1, Use Spawn Probability=false)
  ]
  particleSpawnScript [
   &InitializeParticle(
    Lifetime=5
    Position Offset=<0 -5.5 0>
    Sprite Size Mode=::NewEnumerator3
    Uniform Sprite Size=65
    UsePositionOffset=true
    Write Material Random=true
   ) {module=InitializeParticle001}
   &DynamicMaterialParameters(
    Index 0 Param 1=2.5
    Module.Param0WriteEnabled=true
    Module.Param1WriteEnabled=true
    Module.Param2WriteEnabled=true
    Module.Param3WriteEnabled=true
   )
  ]
  particleUpdateScript [
   &ParticleState(Kill Particles When Lifetime Has Elapsed=false)
   &CameraOffset()
   &ScaleColor(
    Scale Alpha=&Lerp_Float(
     A=1
     B=&Waveform("Waveform [1]"=::NewEnumerator9, "[1] Frequency"=10, "[1] Amplitude Min/Max"=<1 1.25>)
     Alpha=&MakeCustomFloatFromBool(Bool=&SetBoolByFloatComparison(A=$System.Music_Reactive_FX_Preview, B=0), True Float=1, False Float=0)
    )
    Scale RGB=&Multiply_VectorByFloat(
     Vector=&MakeVectorFromLinearColorRGB(Color=&FNTimeOfDayColor(
       Night Color=<rgba 0.0416 0.0016 1 1>
       Morning Color=<rgba 1 0.6236 0.1387 1>
       Day Color=<rgba 1 0.6236 0.1387 1>
       Evening Color=<rgba 1 0.6236 0.1387 1>
      ))
     Float=&Lerp_Float(
      A=1
      B=&Waveform("Waveform [1]"=::NewEnumerator9, "[1] Frequency"=10, "[1] Amplitude Min/Max"=<1 1.5>)
      Alpha=&MakeCustomFloatFromBool(Bool=&SetBoolByFloatComparison(A=$System.Music_Reactive_FX_Preview, B=0), True Float=1, False Float=0)
     )
    )
    ScaleA=true
    ScaleRGB=true
   )
   &ScaleSpriteSize(Scale Factor=&Vector2DFromFloat(Value=&Lerp_Float(
      A=1
      B=&Waveform("Waveform [1]"=::NewEnumerator9, "[1] Frequency"=10, "[1] Amplitude Min/Max"=<1 1.25>)
      Alpha=&MakeCustomFloatFromBool(Bool=&SetBoolByFloatComparison(A=$System.Music_Reactive_FX_Preview, B=0), True Float=1, False Float=0)
     )))
  ]
  rendererClasses [&NiagaraSpriteRendererProperties]
  renderers [{rendererClass=&NiagaraSpriteRendererProperties}]
 }
 Heart {
  emitterUpdateScript [
   &EmitterState(
    Inactive Response=::NewEnumerator1
    Life Cycle Mode=::NewEnumerator1
    Loop Behavior=::NewEnumerator1
    Loop Duration Mode=::NewEnumerator1
    UseLoopDelay=false
   )
   &SpawnBurst_Instantaneous(Use Spawn Probability=false, UseLoopCountLimit=false)
  ]
  particleSpawnScript [
   &InitializeParticle(
    Color Mode=::NewEnumerator0
    Lifetime=5
    Mesh Scale Mode=::NewEnumerator3
    Mesh Uniform Scale=5
    Position Offset=<10 0 0>
    Sprite Size Mode=::NewEnumerator3
    Uniform Sprite Size=100
    UsePositionOffset=false
    Write Material Random=true
   )
   &InitialMeshOrientation(Mesh Axis To Orient=<0 0.25 0>, Mesh Orientation Mode=::NewEnumerator2, Orientation Vector=<0 1 0>, Use Rotation Vector=false)
  ]
  particleUpdateScript [
   &ParticleState(Kill Particles When Lifetime Has Elapsed=false, Let Infinitely Lived Particles Die When Emitter Deactivates=true)
   &DynamicMaterialParameters(
    Index 0 Param 1=&FN_FrontEndFloatModifier(Gameplay=0, Front End=1)
    Index 0 Param 2=1
    Module.Index0Param0Write=true
    Module.Index0Param1Write=true
    Module.Index0Param2Write=false
    Module.Index0Param3Write=false
   )
  ]
  rendererClasses [&NiagaraMeshRendererProperties]
  renderers [{rendererClass=&NiagaraMeshRendererProperties, rendererIndex=0}]
 }
 Heart_Low_End {
  emitterUpdateScript [
   &EmitterState(
    Inactive Response=::NewEnumerator1
    Life Cycle Mode=::NewEnumerator1
    Loop Behavior=::NewEnumerator1
    Loop Duration Mode=::NewEnumerator1
    UseLoopDelay=false
   )
   &SpawnBurst_Instantaneous(Spawn Count=15, Use Spawn Probability=false, UseLoopCountLimit=false)
  ]
  particleSpawnScript [
   &InitializeParticle(
    Color=<rgba 1 1 1 1.2>
    Sprite Rotation Angle=90
    Sprite Rotation Mode=::NewEnumerator0
    Sprite Size Mode=::NewEnumerator3
    Uniform Sprite Size=4.5
    UsePositionOffset=true
    Write Material Random=true
   )
   /SetVariables(Particles.Position=&AddVectorToPosition(
     Position=$Particles.Position
     Vector=&SelectVectorFromArray(
      Direct Array Index=&ReturnExecIndex()
      Vector Selection Array=DataInterface() {propertyValues {FloatData [
         <7 -8 1>
         <-7 -8 1>
         <10 -8 0>
         <-10 -8 0>
         <0 -8 -6>
         <8 -8 -3>
         <-8 -8 -3>
         <0 -5 -7>
         <4 -5 -8>
         <-4 -5 -8>
         <0 -5 -10>
         <9 -5 -5>
         <-9 -5 -5>
         <3 -5 -1>
         <-3 -5 -1>
        ]}}
     )
    )) {isSetParameters=true, module=SetVariables_#5}
   /SetVariables(Particles.SpriteSize=&Multiply_Vector2D(
     A=$Particles.SpriteSize
     B=&SelectVector2DFromArray(
      Direct Array Index=&ReturnExecIndex()
      Vector 2D Selection Array=DataInterface() {propertyValues {FloatData [<5 5>, <5 5>, <4 4>, <5 4>, <4 4>, <4 4>, <4 4>, <3 4>, <4 4>, <4 4>, <4 4>, <3 3>, <3 3>, <3 3>, <3 3>]}}
     )
    )) {isSetParameters=true, module=SetVariables_#6}
   /SetVariables(Particles.SpriteRotation:NiagaraFloat=&SelectFloatFromArray(
     Direct Array Index=&ReturnExecIndex()
     Float Selection Array=DataInterface() {propertyValues {FloatData [0,0,0,0,135,0,75,39,105,45,0,90,15,0]}}
    )) {isSetParameters=true, module=SetVariables_#7}
  ]
  particleUpdateScript [
   &ParticleState(Kill Particles When Lifetime Has Elapsed=false, Let Infinitely Lived Particles Die When Emitter Deactivates=true)
   &DynamicMaterialParameters(
    Index 0 Param 1=&Multiply_Float(A=0.025, B=$Emitter.Age)
    Index 2 Param 1=&Multiply_Float(A=5, B=$Emitter.Age)
    Index 2 Param 2=&RandomRangeFloat(Minimum=0, Maximum=1)
    Index 2 Param 4=10
    Module.Param0WriteEnabled=true
    Module.Param0WriteEnabled2=true
    Module.Param1WriteEnabled=false
    Module.Param1WriteEnabled2=true
    Module.Param2WriteEnabled=false
    Module.Param2WriteEnabled2=true
    Module.Param3WriteEnabled=true
    Module.Param3WriteEnabled2=true
    Write Parameter Index 2=true
   ) {enabled=false}
   &DynamicMaterialParameters(
    Index 0 Param 1=&Add_Float(
     A=&RandomRangeFloat(Minimum=0, Maximum=1)
     B=&FloatFromCurve(
      FloatCurve=DataInterface() {propertyValues {Curve {keys [
          <at 0 0 RCIM_Cubic>
          {arriveTangent=1.5947, interpMode=RCIM_Cubic, leaveTangent=1.5947, tangentMode=RCTM_User, time=0.2557, value=0.5257}
          <at 0.9 1 RCIM_Cubic>
         ]}}}
      CurveIndex=$Particles.Age
      Scale Curve=0.25
     )
    )
    Module.Param0WriteEnabled=true
    Module.Param1WriteEnabled=true
    Module.Param2WriteEnabled=true
    Module.Param3WriteEnabled=true
   ) {module=DynamicMaterialParameters001}
   &ScaleColor(
    Scale Alpha=&Lerp_Float(
     A=1
     B=&Waveform("Waveform [1]"=::NewEnumerator9, "[1] Frequency"=10, "[1] Amplitude Min/Max"=<1 1.25>)
     Alpha=&MakeCustomFloatFromBool(Bool=&SetBoolByFloatComparison(A=$System.Music_Reactive_FX_Preview, B=0), True Float=1, False Float=0)
    )
    Scale RGB=&Multiply_VectorByFloat(
     Vector=&MakeVectorFromLinearColorRGB(Color=&FNTimeOfDayColor(
       Night Color=<rgba 0.0128 0.0005 0.3073 1>
       Morning Color=<rgba 0.375 0.2034 0.0937 1>
       Day Color=<rgba 0.526 0.3114 0.0932 1>
       Evening Color=<rgba 0.4271 0.2823 0.0956 1>
      ))
     Float=&Lerp_Float(
      A=1
      B=&Waveform("Waveform [1]"=::NewEnumerator9, "[1] Frequency"=10, "[1] Amplitude Min/Max"=<1 1.5>)
      Alpha=&MakeCustomFloatFromBool(Bool=&SetBoolByFloatComparison(A=$System.Music_Reactive_FX_Preview, B=0), True Float=1, False Float=0)
     )
    )
    ScaleA=true
    ScaleRGB=true
   )
  ]
  rendererClasses [&NiagaraSpriteRendererProperties]
  renderers [{rendererClass=&NiagaraSpriteRendererProperties}]
 }
 Hearts {
  emitterUpdateScript [
   &EmitterState()
   &SpawnRate(
    Spawn Probability=&FN_FrontEndFloatModifier(
     Gameplay=&MakeCustomFloatFromBool(Bool=&SetBoolByFloatComparison(A=$System.Music_Reactive_FX_Preview, B=0), True Float=1, False Float=0)
     Front End=1
    )
    SpawnRate=&FN_FrontEndFloatModifier(
     Gameplay=&Multiply_Float(
      A=10
      B=&MakeCustomFloatFromBool(Bool=&SetBoolByFloatComparison(A=$System.Music_Reactive_FX_Preview, B=0), True Float=1, False Float=0)
     )
     Front End=10
    )
    Use Spawn Probability=true
   )
  ]
  particleSpawnScript [
   &InitializeParticle(
    Color=&FNTimeOfDayColor(
     Night Color=<rgba 1.3797 1.2685 2 1>
     Morning Color=&RandomRangeLinearColor(Minimum=<rgba 1 0.8765 0.1496 1>, Maximum=<rgba 2 1.1582 0.3393 1>)
     Day Color=&RandomRangeLinearColor(Minimum=<rgba 1.5 1.3147 0.2243 1>, Maximum=<rgba 3 1.7374 0.5089 1>)
     Evening Color=&RandomRangeLinearColor(Minimum=<rgba 1 0.8765 0.1496 1>, Maximum=<rgba 2 1.1582 0.3393 1>)
    )
    Lifetime Max=0.6
    Lifetime Min=0.4
    Lifetime Mode=::NewEnumerator1
    Sprite Rotation Angle=&Lerp_Float(
     A=&RandomRangeFloat(Minimum=-30, Maximum=30)
     B=0
     Alpha=&FNTimeOfDayScalar(Night Scale=1, Morning Scale=0, Day Scale=0, Evening Scale=0)
    )
    Sprite Rotation Mode=::NewEnumerator0
    Sprite Size Mode=::NewEnumerator1
    Sprite UV Mode=::NewEnumerator1
    Uniform Sprite Size Max=12
    Uniform Sprite Size Min=9
    UsePositionOffset=false
    Write Material Random=true
   )
   &SocketLocation(
    OverrideRandomSeed=false
    SamplePosOffset=false
    Socket Reader=$System.DimeBlanket_SocketReader
    Socket Sampling Mode=::NewEnumerator1
    VelocitySpeedLimit=false
   )
   &ShapeLocation(MaskSpawnGroup=false, Override Seed=false, Sphere Radius=8, UseOwnerScale=true)
   &AddVelocity(
    Module.Use Falloff Distance=false
    Override Seed=false
    Velocity Mode=::NewEnumerator1
    Velocity Origin=$Output.SocketLocation.SampledPosition
    Velocity Speed=&Lerp_Float(
     A=8
     B=6
     Alpha=&FNTimeOfDayScalar(Night Scale=1, Morning Scale=0, Day Scale=0, Evening Scale=0)
    )
   )
   &CameraOffset(Camera Offset Amount=&FN_FrontEndFloatModifier(Gameplay=7, Front End=10))
  ]
  particleUpdateScript [
   &ParticleState()
   &SolveForcesAndVelocity(Clamp Velocity=false, Limit Acceleration=false)
   &DynamicMaterialParameters(
    Index 0 Param 1=&FloatFromCurve(
     FloatCurve=DataInterface() {propertyValues {Curve {keys [{interpMode=RCIM_Cubic, tangentMode=RCTM_User, time=0.25, value=0}, {interpMode=RCIM_Cubic, tangentMode=RCTM_User, time=1, value=1}]}}}
     Scale Curve=&RandomRangeFloat(Minimum=0.9, Maximum=1.1)
    )
    Module.Index0Float0Write=true
    Module.Index0Float1Write=true
    Module.Index0Float2Write=true
    Module.Index0Float3Write=true
   )
   &ScaleSpriteSize(
    Uniform Curve Scale=&Lerp_Float(
     A=1
     B=0.8
     Alpha=&FNTimeOfDayScalar(Night Scale=1, Morning Scale=0, Day Scale=0, Evening Scale=0)
    )
    Uniform Curve Sprite Scale=DataInterface() {propertyValues {Curve {keys [
        {arriveTangent=2.4668, interpMode=RCIM_Cubic, leaveTangent=2.4668, tangentMode=RCTM_User, time=0, value=0}
        {tangentMode=RCTM_User, time=1, value=1}
       ]}}}
   )
   &ScaleSpriteSize(
    Uniform Curve Scale=&Multiply_Float(
     A=&RandomRangeFloat(Minimum=0.7, Maximum=1)
     B=&MakeCustomFloatFromBool(
      Bool=$User.MusicReactive
      True Float=1
      False Float=&Multiply_Float(
       A=&FN_FrontEndFloatModifier(Gameplay=1.2, Front End=1.5)
       B=&ClampFloat(Float=$NPC.FortniteNPC.EventAudio_Amplitude, Min=0.3, Max=0.8)
      )
     )
    )
    Uniform Curve Sprite Scale=DataInterface() {propertyValues {Curve {keys [{tangentMode=RCTM_User, time=0, value=0.3}, <at 0.1 1.2>, <at 0.2 1>, {tangentMode=RCTM_User, time=1, value=1}]}}}
   ) {enabled=false, module=ScaleSpriteSize001}
   &ScaleColor(
    Scale Alpha=&Multiply_Float(
     A=&Multiply_Float(
      A=&FloatFromCurve(
       FloatCurve=DataInterface() {propertyValues {Curve {keys [<at 0 1 RCIM_Cubic>,<at 1 0>]}}}
       Scale Curve=&RandomRangeFloat(Minimum=2, Maximum=3)
      )
      B=&MakeFloatFromLinearColor(LinearColor=$System.TODColorMultiplier, Channel:ENiagaraLinearColor_Channels=::NewEnumerator3)
     )
     B=&FN_FrontEndFloatModifier(
      Gameplay=&MakeCustomFloatFromBool(
       Bool=$User.MusicReactive
       True Float=1
       False Float=&MakeCustomFloatFromBool(
        Bool=&SetBoolByFloatComparison(A=$System.Music_Reactive_FX_Preview, B=0)
        True Float=&Multiply_Float(
         A=&Lerp_Float(
          A=&FN_FrontEndFloatModifier(
           Gameplay=&FN_PickFloatDependOnPlatform(TargetPlatformActive=1, TargetPlatformNoActive=0.8)
           Front End=1
          )
          B=&FN_FrontEndFloatModifier(
           Gameplay=&FN_PickFloatDependOnPlatform(TargetPlatformActive=3, TargetPlatformNoActive=1.5)
           Front End=1
          )
          Alpha=$System.Music_Reactive_FX_Preview
         )
         B=&FNTimeOfDayScalar(Night Scale=0.4, Morning Scale=0.75, Day Scale=0.75, Evening Scale=0.75)
        )
        False Float=0
       )
      )
      Front End=1
     )
    )
    ScaleA=true
    ScaleRGB=true
   ) {module=ScaleColor001}
   /SetVariables(
    Particles.CenterFlareColor:LinearColor=&MultiplyLinearColor(
     Linear Color=$Particles.Color
     Scale Factor=&MakeVector4_FromLinearColor(LinearColor=<rgba 3.4218 8.799 10 1>)
    )
    Particles.CenterFlareSize:Vector2f=&Multiply_Vector2(A=$Particles.SpriteSize, B=&Vector2DFromFloat(Value=0.5))
    Particles.CoronaFlareColor:LinearColor=&MultiplyLinearColor(
     Linear Color=$Particles.Color
     Scale Factor=&MakeVector4_FromLinearColor(LinearColor=<rgba 0.15 0.15 0.15 1>)
    )
    Particles.CoronaFlareSize:Vector2f=&Multiply_Vector2(A=$Particles.Initial.SpriteSize, B=&Vector2DFromFloat(Value=0.75))
    Particles.FlickerFlareColor:LinearColor=&MultiplyLinearColor(
     Linear Color=$Particles.Color
     Scale Factor=&MakeVector4_FromLinearColor(LinearColor=<rgba 1 0.9072 0.9536 1>)
    )
   ) {isSetParameters=true}
  ]
  rendererClasses [&NiagaraSpriteRendererProperties]
  renderers [
   {rendererClass=&NiagaraSpriteRendererProperties}
   {rendererClass=&NiagaraSpriteRendererProperties, rendererIndex=1}
   {rendererClass=&NiagaraSpriteRendererProperties, rendererIndex=2}
  ]
 }
 InnerHeart_Low_End {
  emitterUpdateScript [
   &EmitterState(
    Inactive Response=::NewEnumerator1
    Life Cycle Mode=::NewEnumerator1
    Loop Behavior=::NewEnumerator1
    Loop Duration Mode=::NewEnumerator1
    UseLoopDelay=false
   )
   &SpawnBurst_Instantaneous(Use Spawn Probability=false, UseLoopCountLimit=false)
  ]
  particleSpawnScript [
   &InitializeParticle(
    Color=<rgba 1 1 1 1.2>
    Position Offset=<0 0 -5>
    Sprite Rotation Mode=::NewEnumerator0
    Sprite Size Mode=::NewEnumerator3
    Uniform Sprite Size=5
    UsePositionOffset=true
    Write Material Random=true
   )
   /SetVariables(Particles.Position=&AddVectorToPosition(
     Position=$Particles.Position
     Vector=&SelectVectorFromArray(
      Direct Array Index=&ReturnExecIndex()
      Vector Selection Array=DataInterface() {propertyValues {FloatData [
         <0 -5 0>
         <-8 -8 5>
         <10 -8 1>
         <-10 -8 1>
         <0 -5 -1.5>
         <8 -5 -5>
         <-8 -5 -5>
         <0 -5 -9>
         <4 -5 -10>
         <-4 -5 -10>
         <0 -5 -12>
         <7 -5 -5>
         <-7 -5 -5>
         <3 -5 -1>
         <-3 -5 -1>
        ]}}
     )
    )) {isSetParameters=true, module=SetVariables_#5}
   /SetVariables(Particles.SpriteSize=&Multiply_Vector2D(
     A=$Particles.SpriteSize
     B=&SelectVector2DFromArray(
      Direct Array Index=&ReturnExecIndex()
      Vector 2D Selection Array=DataInterface() {propertyValues {FloatData [<3 3>, <3 3.3>, <5 4>, <5 4>, <4 4>, <4 4>, <4 4>, <3 4>, <4 4>, <1 1>, <4 4>, <3 3>, <3 3>, <3 3>, <3 3>]}}
     )
    )) {isSetParameters=true, module=SetVariables_#6}
   /SetVariables(Particles.SpriteRotation:NiagaraFloat=&SelectFloatFromArray(
     Direct Array Index=&ReturnExecIndex()
     Float Selection Array=DataInterface() {propertyValues {FloatData [0,25,-25,45,135,160,75,39,105,45,135,90,15,0]}}
    )) {isSetParameters=true, module=SetVariables_#7}
  ]
  particleUpdateScript [
   &ParticleState(Kill Particles When Lifetime Has Elapsed=false, Let Infinitely Lived Particles Die When Emitter Deactivates=true)
   &DynamicMaterialParameters(
    Index 0 Param 1=&FN_FrontEndFloatModifier(Gameplay=0, Front End=1)
    Module.Param0WriteEnabled=true
    Module.Param1WriteEnabled=false
    Module.Param2WriteEnabled=false
    Module.Param3WriteEnabled=true
   )
   &DynamicMaterialParameters(Write Parameter Index 0=false) {module=DynamicMaterialParameters001}
   &ScaleColor(
    Scale Alpha=&Lerp_Float(
     A=1
     B=&Waveform("Waveform [1]"=::NewEnumerator9, "[1] Frequency"=10, "[1] Amplitude Min/Max"=<1 1.25>)
     Alpha=&MakeCustomFloatFromBool(Bool=&SetBoolByFloatComparison(A=$System.Music_Reactive_FX_Preview, B=0), True Float=1, False Float=0)
    )
    Scale RGB=&Multiply_VectorByFloat(
     Vector=&MakeVectorFromLinearColorRGB(Color=&FNTimeOfDayColor(
       Night Color=<rgba 0.0128 0.0005 0.3073 1>
       Morning Color=<rgba 0.375 0.2034 0.0937 1>
       Day Color=<rgba 0.526 0.3114 0.0932 1>
       Evening Color=<rgba 0.4271 0.2823 0.0956 1>
      ))
     Float=&Lerp_Float(
      A=1
      B=&Waveform("Waveform [1]"=::NewEnumerator9, "[1] Frequency"=10, "[1] Amplitude Min/Max"=<1 1.5>)
      Alpha=&MakeCustomFloatFromBool(Bool=&SetBoolByFloatComparison(A=$System.Music_Reactive_FX_Preview, B=0), True Float=1, False Float=0)
     )
    )
    ScaleA=true
    ScaleRGB=true
   )
  ]
  rendererClasses [&NiagaraSpriteRendererProperties]
  renderers [{rendererClass=&NiagaraSpriteRendererProperties}]
 }
}
summary {
 systemName=NS_DimeBlanketKnot_Backpack
 emitters [
  {emitterName=AirGlow, rendererClasses [&NiagaraSpriteRendererProperties]}
  {emitterName=Hearts, rendererClasses [&NiagaraSpriteRendererProperties]}
  {emitterName=Heart, rendererClasses [&NiagaraMeshRendererProperties]}
  {emitterName=Heart_Low_End, rendererClasses [&NiagaraSpriteRendererProperties]}
  {emitterName=InnerHeart_Low_End, rendererClasses [&NiagaraSpriteRendererProperties]}
 ]
 userVariables [
  User.bIsFrontEnd {type=NiagaraBool}
  User.bIsFrontEndPreview {type=NiagaraBool}
  User.Lifetime_Scale {type=NiagaraFloat, defaultValue=0, description=""}
  User.MusicReactive {type=NiagaraBool}
 ]
}
