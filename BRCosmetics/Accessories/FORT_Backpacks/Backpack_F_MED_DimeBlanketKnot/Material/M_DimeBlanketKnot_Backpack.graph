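# Unlit alpha-composite material for the DimeBlanketKnot back bling, rendering a volumetric heart cloud with sparkles.
# A custom HLSL raymarch samples a noise-distorted pseudo-volume heart SDF with baked shadow and AO, tinted by time-of-day curves.
# Pivot-anchored screen-space UVs map a heart mask and animated star layers, with audio input for reactivity.
# Depth test is off, so rays clip against scene depth; platform switches scale step count and jitter.
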Material /BRCosmetics/Accessories/FORT_Backpacks/Backpack_F_MED_DimeBlanketKnot/Material/M_DimeBlanketKnot_Backpack
graph {
 properties {
  bDisableDepthTest=true
  bUsedWithMeshDeformer=true
  bUsedWithNiagaraMeshParticles=true
  bUsedWithSkeletalMesh=true
  bUsedWithStaticLighting=true
  blendMode=BLEND_AlphaComposite
  pixelDepthOffsetMode=PDOM_Legacy
  shadingModel=MSM_Unlit
  usedShadingModels=Unlit
 }
 nodes [
  n1 = ScalarParameter() {defaultValue=6, parameterName=ScaleU}
  n2 = ScalarParameter() {defaultValue=6, parameterName=Scalev}
  n3 = NamedRerouteUsage()
  n4 = Multiply(A=Multiply(A=AppendVector(A=@n1, B=@n2), B=Add(A=@n3)), B=Sign(None=@n3))
  Abs(None=@n4)
  n5 = CurveAtlasRowParameter(CurveTime=NamedRerouteUsage()) {
   atlas=/BRCosmetics/Characters/Player/Female/Medium/Bodies/F_MED_DimeBlanketGrace/Materials/CA_DimeBlanket
   curve=/BRCosmetics/Characters/Player/Female/Medium/Bodies/F_MED_DimeBlanketGrace/Materials/C_DimeBlanketTODTransition_Lighting
   parameterName=TODCurve_Lighting
  }
  Add(A=@n5, B=NamedRerouteUsage())
  n6 = MaterialFunctionCall(Float3=TransformPosition(None=MaterialFunctionCall() {materialFunction=&ObjectPivotPoint}.Object Pivot Location) {transformSourceType=TRANSFORMPOSSOURCE_World}) {materialFunction=&BreakOutFloat3Components}
  n7 = ComponentMask(None=Multiply(A=ObjectBounds()) {constB=0.5}) {b=true}
  n8 = TransformPosition(None=MaterialFunctionCall(X=@n6.R, Y=@n6.R, Z=Add(A=@n6.B, B=@n7)) {materialFunction=&MakeFloat3}.Result) {transformType=TRANSFORMPOSSOURCE_World}
  n9 = WorldPosition()
  n10 = Multiply(A=Normalize(VectorInput=Subtract(A=@n8, B=@n9.XYZ))) {constB=-1}
  Add(A=@n10)
  n11 = MaterialFunctionCall() {materialFunction=&ObjectPivotPoint}
  n12 = TransformPosition(None=VectorParameter() {parameterName=PositionOffset, defaultValue=<rgba 0 -7 0 1>}.RGB) {transformType=TRANSFORMPOSSOURCE_World}
  Add(A=@n11.Object Pivot Location, B=@n12)
  n13 = ScreenPosition()
  n14 = Constant2Vector() {desc="The 'V' of screen space coordinate is opposite of the \"Y\" of clip space coordinate", g=-1, r=1}
  n15 = MaterialFunctionCall(World Position=@n12) {materialFunction=&TransformToClipSpace}
  n16 = MaterialFunctionCall(Float2=ViewSize()) {materialFunction=&BreakOutFloat2Components}
  n17 = Multiply(
   A=Subtract(A=ConstantBiasScale(None=@n13.ViewportUV) {bias=-0.5, scale=2}, B=Multiply(A=@n14, B=@n15.Clip Space XY))
   B=AppendVector(A=Constant() {r=1}, B=Divide(A=@n16.G, B=@n16.G))
  )
  n18 = Multiply(A=@n17, B=@n4)
  n19 = ComponentMask(None=TransformPosition(None=@n11.Object Pivot Location) {transformSourceType=TRANSFORMPOSSOURCE_World, transformType=TRANSFORMPOSSOURCE_View}) {b=true}
  n20 = Constant() {r=120}
  n21 = Add(A=Divide(A=Multiply(A=@n18, B=@n19), B=@n20), B=Constant2Vector() {g=0.5, r=0.5})
  n22 = Saturate(None=@n21)
  Clamp(None=@n22)
  MaterialExpressionTextureObjectParameter_6 = TextureObjectParameter() {
   desc=This texture is still at lifting aura folder
   parameterName=Gradient_SDF
   samplerType=SAMPLERTYPE_LinearColor
   texture=/Game/Animation/Props/Gadgets_LiftingAura/Textures/heartSDFGrad_v01_Volume
  }
  MaterialExpressionNamedRerouteUsage_53 = NamedRerouteUsage()
  MaterialExpressionNamedRerouteUsage_52 = NamedRerouteUsage()
  MaterialExpressionNamedRerouteUsage_51 = NamedRerouteUsage()
  MaterialExpressionNamedRerouteUsage_54 = NamedRerouteUsage()
  MaterialExpressionNamedRerouteUsage_55 = NamedRerouteUsage()
  n23 = TextureObjectParameter() {
   parameterName=Shadow
   samplerType=SAMPLERTYPE_LinearColor
   texture=/Game/Global/Materials/MaterialGym/Materials/RayMarchedSDFVolume/Textures/M_VolumeEncode_Global_Baked
  }
  MaterialExpressionReroute_4 = Reroute(None=@n23)
  MaterialExpressionReroute_5 = Reroute(None=ScalarParameter() {defaultValue=14, parameterName=densityGain})
  MaterialExpressionScalarParameter_24 = ScalarParameter() {defaultValue=0.3, parameterName=Shadow Power}
  n24 = CurveAtlasRowParameter(CurveTime=NamedRerouteUsage()) {
   atlas=&CA_DimeBlanket
   curve=/BRCosmetics/Characters/Player/Female/Medium/Bodies/F_MED_DimeBlanketGrace/Materials/C_DimeBlanketTODTransition_AmbientColor
   defaultValue=1
   parameterName=TODCurve_Ambient
  }
  MaterialExpressionNamedRerouteDeclaration_5 = NamedRerouteDeclaration(None=@n24) {name=AmbientColor, nodeColor=<rgba 0 1 0.4141 1>}
  n25 = VectorParameter() {parameterName=AtmosphereColorD, defaultValue=<rgba 0.9418 0.9483 1.5 1>}
  n26 = VectorParameter() {parameterName=AtmosphereColorN, defaultValue=<rgba 0.2537 0.3181 1 1>}
  n27 = NamedRerouteUsage()
  n28 = LinearInterpolate(A=@n5, B=LinearInterpolate(A=@n5, B=LinearInterpolate(A=@n25.RGB, B=@n26.RGB, Alpha=@n27)), Alpha=NamedRerouteUsage())
  MaterialExpressionReroute_7 = Reroute(None=EyeAdaptationInverse(
    LightValueInput=NamedRerouteDeclaration(None=@n28) {name=LightColor, nodeColor=<rgba 0 1 0.2266 1>}
    AlphaInput=ScalarParameter() {parameterName=EyeAdaptionInverse}
   ).EyeAdaptationInverse)
  MaterialExpressionSkyAtmosphereLightDirection_2 = SkyAtmosphereLightDirection()
  MaterialExpressionTextureObjectParameter_5 = TextureObjectParameter() {parameterName=Noise, texture=/Game/Global/Materials/MaterialGym/Materials/RayMarchedSDFVolume/Textures/T_Volume_Global_PerlinWorley}
  MaterialExpressionTime_1 = Time()
  MaterialExpressionMultiply_38 = Multiply(A=@MaterialExpressionTime_1, B=ScalarParameter() {defaultValue=0.06, parameterName=NoiseTime}) {constB=0.015}
  MaterialExpressionScalarParameter_26 = ScalarParameter() {defaultValue=0.06, parameterName=NoiseStrength}
  MaterialExpressionScalarParameter_27 = ScalarParameter() {defaultValue=0.31, parameterName=SphereRadius}
  MaterialExpressionVectorParameter_17 = VectorParameter() {parameterName=SpherePosition, defaultValue=<rgba 0.5 0.2 0.475 1>}
  MaterialExpressionScalarParameter_28 = ScalarParameter() {
   defaultValue=0.666
   desc=0-1 Value. 1 is linear falloff from edge to sphere origin. 0 is hard edge sphere
   parameterName=NormalizedSphereFalloff
  }
  n29 = TextureObjectParameter() {
   parameterName=Stream
   samplerType=SAMPLERTYPE_LinearColor
   texture=/Game/Global/Materials/MaterialGym/Materials/RayMarchedSDFVolume/Textures/VT_Heart_Global_SDF_V01
  }
  MaterialExpressionReroute_8 = Reroute(None=@n29)
  MaterialExpressionScalarParameter_29 = ScalarParameter() {defaultValue=2, parameterName=AO Power}
  MaterialExpressionScalarParameter_30 = ScalarParameter() {defaultValue=0.4, parameterName=Inner Softness Distance}
  MaterialExpressionScalarParameter_31 = ScalarParameter() {defaultValue=1, parameterName=Density Darken Min}
  MaterialExpressionScalarParameter_32 = ScalarParameter() {defaultValue=3, parameterName=Density Darken Power}
  n30 = Custom!(
   VolumeTexture=@MaterialExpressionTextureObjectParameter_6
   rayStart=@MaterialExpressionNamedRerouteUsage_53
   numSteps=@MaterialExpressionNamedRerouteUsage_52
   cameraVector=@MaterialExpressionNamedRerouteUsage_51
   stepSize=@MaterialExpressionNamedRerouteUsage_54
   finalDelta=@MaterialExpressionNamedRerouteUsage_55
   ShadowTexture=@MaterialExpressionReroute_4
   densityGain=@MaterialExpressionReroute_5
   shadowGain=@MaterialExpressionScalarParameter_24
   ambientColor=@MaterialExpressionNamedRerouteDeclaration_5
   lightColor=@MaterialExpressionReroute_7
   localLightDirection=@MaterialExpressionSkyAtmosphereLightDirection_2
   noise=@MaterialExpressionTextureObjectParameter_5
   timeAnim=@MaterialExpressionMultiply_38
   noiseStrength=@MaterialExpressionScalarParameter_26
   SphereRadius=@MaterialExpressionScalarParameter_27
   SpherePosition=@MaterialExpressionVectorParameter_17.RGB
   NormalizedSphereFalloff=@MaterialExpressionScalarParameter_28
   Streaming2DTex=@MaterialExpressionReroute_8
   AO_power=@MaterialExpressionScalarParameter_29
   innerSoftnessDistance=@MaterialExpressionScalarParameter_30
   DensityDarkenMin=@MaterialExpressionScalarParameter_31
   DensityDarkenPower=@MaterialExpressionScalarParameter_32
  ) {
   code="\r\n\r\nfloat theta = dot(normalize(localLightDirection), normalize(cameraVector));\r\nfloat3 cameraVectorOld = cameraVector;\r\ncameraVector *= stepSize;\r\n\r\n\r\nstruct Func {\r\n\r\n// the Henyey-Greenstein phase function\r\nfloat phase( float g,  float cos_theta)\r\n{\r\n    float denom = 1.0 + g * g - 2.0 * g * cos_theta;\r\n    return 1.0 / (4 * 3.14159) * (1.0 - g * g) / (denom * sqrt(denom));\r\n}\r\n\r\nfloat remapValueRange(float value, float low1, float high1, float low2,  float high2)\r\n{\r\n    return low2 + (value - low1) * (high2 - low2) / (high1 - low1);\r\n}\r\n\r\nfloat sphereDistance(float3 origin, float3 position, float radius){\r\n\r\n    //vector from position to sphere centre\r\n    float3 relativePosition = origin-position;\r\n\r\n    return length(relativePosition)-radius;\r\n}\r\n\r\n};\r\n\r\n\r\nfloat accumdist = 0;\r\nfloat current_density = 0;\r\nfloat transmittance = 1.0;\r\nfloat3 lightenergy = 0;\r\n\r\nFunc func;\r\n\r\nfloat height = GetPrimitiveData(Parameters.PrimitiveId).LocalObjectBoundsMax.z;\r\nfloat3 tim = (float3(timeAnim, timeAnim, timeAnim));\r\n\r\nfor (int i = 0; i < numSteps; i++){\r\n    //float4 noiseSample = noise.SampleLevel( noiseSampler, (rayStart + float3(0,-timeAnim*0.6, -timeAnim)/2),0);\r\n    float4 noiseSample = PseudoVolumeTexture( noise, noiseSampler, (rayStart + float3(0,-timeAnim*0.6, -timeAnim)/2), float2(16, 8), 128);\r\n    //float4 cursample = VolumeTexture.SampleLevel( VolumeTextureSampler, saturate(rayStart-(noiseSample.bga-0.5)*noiseStrength),0);\r\n    // float4 PseudoVolumeTexture(Texture2D Tex, SamplerState TexSampler, float3 inPos, float2 xysize, float numframes,\r\n\t//uint mipmode = 0, float miplevel = 0, float2 InDDX = 0, float2 InDDY = 0)\r\n    //float4 cursample = PseudoVolumeTexture( Streaming2DTex, Streaming2DTexSampler, saturate(rayStart-(noiseSample.bga-0.5)*noiseStrength), float2(128, 128), 121 );\r\n    float3 pos = saturate(rayStart-(noiseSample.bga-0.5)*noiseStrength);\r\n    float4 cursample = PseudoVolumeTexture( Streaming2DTex, Streaming2DTexSampler, pos, float2(11, 11), 121 );\r\n    float dens = func.remapValueRange(cursample.r, innerSoftnessDistance, 0.56, 1, 0);\r\n    float darken =  func.remapValueRange(saturate(pow(1.0-dens,DensityDarkenPower)), 0, 1,DensityDarkenMin, 1);\r\n    //dens *= saturate(func.remapValueRange(func.sphereDistance(SpherePosition, pos, SphereRadius), -SphereRadius*NormalizedSphereFalloff, 0.0, 0, 1));\r\n    //float2 curShadowsample = ShadowTexture.SampleLevel( VolumeTextureSampler, saturate(rayStart-(noiseSample.bga-0.5)*noiseStrength),0).rg;\r\n    float2 curShadowsample = PseudoVolumeTexture( ShadowTexture, ShadowTextureSampler, saturate(rayStart-(noiseSample.bga-0.5)*noiseStrength),float2(8, 8), 64).rg;\r\n    if(dens > 0.001){\r\n\r\n        current_density = saturate(dens)  * densityGain    *  stepSize;\r\n        float t_i = 1- exp(-current_density  );\r\n        \r\n        lightenergy += lightColor * t_i * transmittance * pow(curShadowsample.r, shadowGain) * darken;// * func.phase(0.1,theta)*1/stepSize ;\r\n        lightenergy += ambientColor * t_i * transmittance * pow(saturate(curShadowsample.g), AO_power) ;\r\n        //lightenergy += 5*pow(saturate(func.remapValueRange(func.sphereDistance(SpherePosition+float3(0,0.1,0), rayStart, SphereRadius), -SphereRadius*NormalizedSphereFalloff, 0.0, 0.1, 0)),2);\r\n        transmittance *= 1- t_i;  \r\n    Volume = cursample;    \r\n        \r\n    }\r\n    rayStart += cameraVector;\r\n}\r\n\r\n\r\n// final step to remove artefact\r\n{\r\n    rayStart -= cameraVector;\r\n    rayStart += cameraVectorOld*finalDelta;\r\n\r\n    //float4 noiseSample = noise.SampleLevel( noiseSampler, (rayStart + float3(0,-timeAnim*0.6, -timeAnim)/2),0);\r\n    float4 noiseSample = PseudoVolumeTexture( noise, noiseSampler, (rayStart + float3(0,-timeAnim*0.6, -timeAnim)/2), float2(16, 8), 128);\r\n    //float4 cursample = VolumeTexture.SampleLevel( VolumeTextureSampler, saturate(rayStart-(noiseSample.bga-0.5)*noiseStrength),0);\r\n    float3 pos = saturate(rayStart-(noiseSample.bga-0.5)*noiseStrength);\r\n    float4 cursample = PseudoVolumeTexture( Streaming2DTex, Streaming2DTexSampler,  pos, float2(11, 11), 121 );\r\n    float dens = func.remapValueRange(cursample.a, innerSoftnessDistance, 0.56, 1, 0);\r\n    float darken =  func.remapValueRange(saturate(pow(1.0-dens,DensityDarkenPower)), 0, 1,DensityDarkenMin, 1);\r\n    //dens *= saturate(func.remapValueRange(func.sphereDistance(SpherePosition, pos, SphereRadius), -SphereRadius*NormalizedSphereFalloff, 0.0, 0, 1));\r\n    //float2 curShadowsample = ShadowTexture.SampleLevel( VolumeTextureSampler,saturate(rayStart-(noiseSample.bga-0.5)*noiseStrength),0).rg;\r\n    float2 curShadowsample = PseudoVolumeTexture( ShadowTexture, ShadowTextureSampler, saturate(rayStart-(noiseSample.bga-0.5)*noiseStrength),float2(8, 8), 64).rg;\r\n    if(dens > 0.001){\r\n\r\n        current_density = saturate(dens)  * densityGain    *  stepSize;\r\n        float t_i = 1- exp(-current_density  );\r\n        \r\n        lightenergy += lightColor * t_i * transmittance * pow(curShadowsample.r, shadowGain) * darken;// * func.phase(0.1,theta)*1/stepSize ;\r\n        lightenergy += ambientColor * t_i * transmittance *  pow(saturate(curShadowsample.g), AO_power);\r\n        transmittance *= 1- t_i;  \r\n        \r\n        \r\n    }\r\n}\r\n\r\nreturn float4( lightenergy, 1-transmittance);\r\n\r\n\r\n\r\n\r\n\r\n\r\n"
   description=DensityRayMarch
   outputType=CMOT_Float4
   showCode=true
   additionalOutputs [{outputName=Volume, outputType=CMOT_Float4}]
   inputs [
    {
     inputName=VolumeTexture
     input {expression=&self:MaterialExpressionTextureObjectParameter_6}
    }
    {
     inputName=rayStart
     input {expression=&self:MaterialExpressionNamedRerouteUsage_53}
    }
    {
     inputName=numSteps
     input {expression=&self:MaterialExpressionNamedRerouteUsage_52}
    }
    {
     inputName=cameraVector
     input {expression=&self:MaterialExpressionNamedRerouteUsage_51}
    }
    {
     inputName=stepSize
     input {expression=&self:MaterialExpressionNamedRerouteUsage_54}
    }
    {
     inputName=finalDelta
     input {expression=&self:MaterialExpressionNamedRerouteUsage_55}
    }
    {
     inputName=ShadowTexture
     input {expression=&self:MaterialExpressionReroute_4}
    }
    {
     inputName=densityGain
     input {expression=&self:MaterialExpressionReroute_5}
    }
    {
     inputName=shadowGain
     input {expression=&self:MaterialExpressionScalarParameter_24}
    }
    {
     inputName=ambientColor
     input {expression=&self:MaterialExpressionNamedRerouteDeclaration_5}
    }
    {
     inputName=lightColor
     input {expression=&self:MaterialExpressionReroute_7}
    }
    {
     inputName=localLightDirection
     input {expression=&self:MaterialExpressionSkyAtmosphereLightDirection_2}
    }
    {
     inputName=noise
     input {expression=&self:MaterialExpressionTextureObjectParameter_5}
    }
    {
     inputName=timeAnim
     input {expression=&self:MaterialExpressionMultiply_38}
    }
    {
     inputName=noiseStrength
     input {expression=&self:MaterialExpressionScalarParameter_26}
    }
    {
     inputName=SphereRadius
     input {expression=&self:MaterialExpressionScalarParameter_27}
    }
    {
     inputName=SpherePosition
     input {
      expression=&self:MaterialExpressionVectorParameter_17
      mask=1
      maskB=1
      maskG=1
      maskR=1
     }
    }
    {
     inputName=NormalizedSphereFalloff
     input {expression=&self:MaterialExpressionScalarParameter_28}
    }
    {
     inputName=Streaming2DTex
     input {expression=&self:MaterialExpressionReroute_8}
    }
    {
     inputName=AO_power
     input {expression=&self:MaterialExpressionScalarParameter_29}
    }
    {
     inputName=innerSoftnessDistance
     input {expression=&self:MaterialExpressionScalarParameter_30}
    }
    {
     inputName=DensityDarkenMin
     input {expression=&self:MaterialExpressionScalarParameter_31}
    }
    {
     inputName=DensityDarkenPower
     input {expression=&self:MaterialExpressionScalarParameter_32}
    }
   ]
  }
  ComponentMask(None=@n30.Volume) {b=true, g=true, r=true}
  Constant() {r=1}
  Custom!(None) {
   code="int3 randpos = int3(Parameters.SvPosition.xy, View.StateFrameIndexMod8);\r\nfloat rand =float(Rand3DPCG16(randpos).x) / 0xffff;\r\nreturn rand.x;"
   showCode=true
  }
  n31 = Divide(
   A=SceneDepth(UVs=MaterialFunctionCall() {materialFunction=&ScreenAlignedUVs}."X 100%, Y 100%").R
   B=ScalarParameter() {defaultValue=1}
  )
  n32 = ScalarParameter() {parameterName=Param_1}
  n33 = Multiply(A=ComponentMask(None=MaterialFunctionCall() {materialFunction=/Game/Animation/Props/Gadgets_LiftingAura/Materials/MaterialFunctions/MF_InstanceBounds}.Instance Bounds Max) {r=true}) {constB=2}
  n34 = MaterialFunctionCall(Vector 3=Transform(None=Constant3Vector() {constant=<rgba 1 0 0 0>}) {transformSourceType=TRANSFORMSOURCE_Instance}) {materialFunction=/Engine/Functions/Engine_MaterialFunctions02/Utility/VectorLength}
  n35 = CameraVectorWS()
  n36 = TransformPosition(None=Constant3Vector() {constant=<rgba 0 0 1 0>}) {transformSourceType=TRANSFORMPOSSOURCE_View, transformType=TRANSFORMPOSSOURCE_TranslatedWorld}
  n37 = Divide(A=Divide(A=Add(A=@n31, B=@n32), B=Multiply(A=@n33, B=@n34.V3 Length) {constB=2}), B=Abs(None=DotProduct(A=@n35, B=@n36)))
  Divide(A=Min(B=@n37) {constA=999}) {constB=500}
  n38 = Custom!(in=@MaterialExpressionTime_1) {
   code="// The below expression will get compiled\r\n// into the output of this node\r\nreturn in;"
   outputType=CMOT_Float1
   showCode=true
   inputs [{
     inputName=in
     input {expression=&self:MaterialExpressionTime_1}
    }]
  }
  Frac(None=@n38)
  n39 = NamedRerouteUsage()
  n40 = LinearInterpolate(A=LinearInterpolate(A=@n39, B=NamedRerouteUsage(), Alpha=NamedRerouteUsage()), B=@n39, Alpha=ScalarParameter() {parameterName=FEAudio})
  n41 = NamedRerouteUsage()
  LinearInterpolate(A=@n40, Alpha=@n41)
  MaterialFunctionCall() {materialFunction=&Scalar_Time_of_Day}
  n42 = CollectionParameter() {
   collection=&FortniteMaterialParameters
   parameterName=FortniteActiveTimeOfDay
  }
  n43 = Power(
   Base=MaterialFunctionCall(Float4=@n42) {materialFunction=&BreakOutFloat4Components}.A
   Exp=ScalarParameter() {defaultValue=1, parameterName=TODPower}
  )
  n44 = NamedRerouteUsage()
  MaterialFunctionCall!(Vector2=AppendVector(A=@n43, B=Constant()), MaximumNumberOfDigits, UVs=@n44, DebugTextLocation RG_UpperRight BA_LowerLeft, Component Spacing) {materialFunction=/Engine/Functions/Engine_MaterialFunctions02/Utility/DebugFloat2Values}
  n45 = Divide(B=NamedRerouteUsage())
  n46 = Multiply(A=Subtract(A=Constant(), B=NamedRerouteUsage()), B=@n45)
  n47 = Multiply(A=Subtract(A=Constant() {r=1}, B=NamedRerouteUsage()), B=@n45)
  n48 = MaterialFunctionCall(Float3=Max(A=@n46, B=@n47)) {materialFunction=&BreakOutFloat3Components}
  n49 = Min(A=@n48.R, B=Min(A=@n48.G, B=@n48.G))
  n50 = Min(A=@n49, B=@n37)
  n51 = MaterialFunctionCall(
   UVs=MaterialFunctionCall!(Position, UVRadialScale) {materialFunction=/Game/Characters/Player/Common/MaterialFunctions/Superlevel/CharacterStabilizedScreenSpaceUVs}.StabilizedUVs
   Texture Scale=Constant() {r=0.3}
  ) {materialFunction=&ScaleUVsByCenter}
  MaterialFunctionCall(Number=@n50, UVs=@n51.UVs) {materialFunction=&DebugScalarValues}
  MaterialFunctionCall() {materialFunction=&ObjectLocalBounds}
  MaterialFunctionCall!(Steps, Plane alignment.) {materialFunction=/Volumetrics/Content/VolumeTextures/MaterialFunctions/VolumeBoxIntersect}
  MaterialFunctionCall() {materialFunction=&DebugTimeSine}
  Multiply()
  n52 = ComponentMask(None=@n30.return) {b=true, g=true, r=true}
  n53 = Multiply(
   A=LinearInterpolate(
    A=ScalarParameter() {defaultValue=2, parameterName=OuterEmissiveMult_Day}
    B=ScalarParameter() {defaultValue=5, parameterName=OuterEmissiveMult_Night}
    Alpha=NamedRerouteUsage()
   )
   B=@n52
  )
  n54 = NamedRerouteUsage()
  n55 = Add(A=ConstantBiasScale(None=@n54) {bias=-0.5, scale=1.4})
  n56 = Sign(None=Multiply(A=@n54, B=NamedRerouteUsage()))
  n57 = SphereMask(
   A=NamedRerouteUsage()
   B=Constant() {r=1}
   Radius=Multiply(A=ScalarParameter() {defaultValue=0.2, parameterName=OuterEmissiveDistance}, B=LinearInterpolate(B=@n55, Alpha=@n56) {constA=1})
  ) {attenuationRadius=0.01, hardnessPercent=0}
  Multiply(A=Add(A=LinearInterpolate(A=@n53, B=@n52, Alpha=@n57), B=NamedRerouteUsage()), B=NamedRerouteUsage())
  NamedRerouteDeclaration(None=ComponentMask(None=@n30.return) {a=true}) {name=TransmittanceMask, nodeColor=<rgba 0 1 0.8359 1>}
  n58 = AtmosphericLightVector()
  n59 = Saturate(None=ConstantBiasScale(None=DotProduct(A=@n58, B=@n10)))
  n60 = ScalarParameter() {defaultValue=1.2, parameterName=SamplePos}
  n61 = ScalarParameter() {defaultValue=0.8, parameterName=TransmittanceRadius}
  n62 = ScalarParameter() {parameterName=TransmittanceHardness}
  NamedRerouteDeclaration(None=Power(Base=SphereMask(A=@n59, B=@n60, Radius=@n61, Hardness=@n62), Exp=ScalarParameter() {defaultValue=5, parameterName=TransmittanceFalloff})) {name=TransmittanceMask_B, nodeColor=<rgba 1 0.4219 0 1>}
  n63 = NamedRerouteUsage()
  n64 = Sign(None=@n63)
  NamedRerouteDeclaration(None=@n64) {name=AudioReactivityMask, nodeColor=<rgba 0 1 0.2734 1>}
  NamedRerouteDeclaration(None=MaterialFunctionCall(Number=@n43, UVs=@n44) {materialFunction=&DebugScalarValues}.Result) {name=Debug, nodeColor=<rgba 0.3828 0 1 1>}
  NamedRerouteDeclaration(None=@n21) {name=ScreenSpaceUV, nodeColor=<rgba 0 1 0.625 1>}
  n65 = DynamicParameter() {paramNames [FEAudio, ReactivitySwitch, Param3, Param4]}
  NamedRerouteDeclaration(None=@n65.ReactivitySwitch) {name=ReactivitySwitch, nodeColor=<rgba 1 0 0.1172 1>}
  n66 = ScalarParameter() {defaultValue=24, parameterName=maxSteps}
  n67 = Multiply(A=@n66, B=ScalarParameter() {defaultValue=0.5, parameterName=Mobile Quality Multipler})
  NamedRerouteDeclaration(None=MaterialFunctionCall("Xb1,PS4"=@n66, "PC-High&&Epic,XSX,PS5"=@n66, "PC-Perf,Android"=@n67, Switch=@n66, "PC-Low(NotPerf)&&Medium"=@n67) {materialFunction=&PlatformSwitch_1D}.Result) {name=MaxSteps, nodeColor=<rgba 0.8516 0 1 1>}
  n68 = MaterialFunctionCall!(Alpha Threshold, Random) {
   desc=This will take one texture sample and miplevel always 0
   materialFunction=/Game/TimeOfDay/CloudsAndStorm/Materials/MaterialFunctions/DitherTemporalAASpatioTemporalBlue
  }
  NamedRerouteDeclaration(None=@n68.Result) {name=BlueNoise, nodeColor=<rgba 1 0 0.7734 1>}
  n69 = CameraPositionWS()
  n70 = Constant()
  n71 = ScalarParameter() {defaultValue=6, parameterName=HorizontalOffset}
  n72 = ScalarParameter() {parameterName=VerticalOffset}
  n73 = Divide(
   A=Add(
    A=TransformPosition(None=@n69) {transformSourceType=TRANSFORMPOSSOURCE_World, transformType=TRANSFORMPOSSOURCE_Instance}
    B=MaterialFunctionCall(X=@n70, Y=@n71, Z=@n72) {materialFunction=&MakeFloat3}.Result
   )
   B=ComponentMask(None=PreSkinnedLocalBounds().Max) {r=true}
  )
  n74 = Multiply(A=@n73) {constB=0.5, desc=The Local Space Pivot is at the center of the cube}
  NamedRerouteDeclaration(None=Add(A=@n74) {constB=0.5}) {name=Local Camera Position, desc=This value can go out of 1, nodeColor=<rgba 0.2891 0 1 1>}
  NamedRerouteDeclaration(None=MaterialFunctionCall(Vector 3=@n74) {materialFunction=&VectorLength}.V3 Length) {name=Camera Distance From Centre of Vol Box, nodeColor=<rgba 0.5 1 0 1>}
  NamedRerouteDeclaration(None=@n73) {name=Normalized Camera Position, nodeColor=<rgba 0 1 0.7891 1>}
  n75 = Transform(None=CameraVectorWS()) {transformSourceType=TRANSFORMSOURCE_World, transformType=TRANSFORM_Instance}
  NamedRerouteDeclaration(None=Multiply(A=Normalize(VectorInput=@n75)) {constB=-1}) {name=Local Camera Vector, nodeColor=<rgba 0 1 0.5781 1>}
  n76 = StaticSwitchParameter(True=@n50, False=@n49) {defaultValue=true, desc="Since material disables depth test, the sorting is done by scenedepth", parameterName=Composite Against Scene Depth}
  NamedRerouteDeclaration(None=@n76) {name=Far Distance t_1, nodeColor=<rgba 1 0 0.3984 1>}
  n77 = MaterialFunctionCall(Float3=Min(A=@n46, B=@n47)) {materialFunction=&BreakOutFloat3Components}
  n78 = Max(A=@n77.G, B=@n77.G)
  NamedRerouteDeclaration(None=Max(A=Max(A=@n77.R, B=@n78)) {constB=0}) {name=Near distance t_0, nodeColor=<rgba 0.875 0 1 1>}
  NamedRerouteDeclaration(None=Divide(A=Constant() {desc="sqrt(1^2+1^2+1^2)", r=1.7321}, B=NamedRerouteUsage())) {name=Step Size, desc=The size is relatived to cube, nodeColor=<rgba 1 0.9844 0 1>}
  n79 = Subtract(A=NamedRerouteUsage(), B=NamedRerouteUsage())
  n80 = Multiply(A=NamedRerouteUsage()) {constB=0.5}
  n81 = Subtract(A=Constant() {desc=Relative 1 unit, r=1}, B=Frac(None=Multiply(A=@n79, B=@n80)))
  n82 = NamedRerouteUsage()
  NamedRerouteDeclaration(None=Multiply(A=Multiply(A=@n81, B=@n82) {constB=0.2}, B=ScalarParameter() {defaultValue=1, parameterName=planeAlignment, sliderMax=1})) {name=Plane offset, nodeColor=<rgba 1 0 0.7734 1>}
  n83 = Multiply(A=NamedRerouteUsage(), B=NamedRerouteUsage())
  NamedRerouteDeclaration(None=@n83) {name=BlueStep, nodeColor=<rgba 0.027 0 1 1>}
  n84 = NamedRerouteUsage()
  n85 = Add(A=@n84, B=NamedRerouteUsage())
  n86 = Max(A=MaterialFunctionCall("Xb1,PS4"=@n84, "PC-High&&Epic,XSX,PS5"=@n84, "PC-Perf,Android"=@n85, Switch=@n84, "PC-Low(NotPerf)&&Medium"=@n85) {materialFunction=&PlatformSwitch_1D}.Result) {constB=0}
  n87 = Add(A=NamedRerouteUsage(), B=Multiply(A=NamedRerouteUsage(), B=@n86)) {desc="The Raymarching here already skip the distance between camera and cube\r\nwhen camera position is over 0-1 space"}
  n88 = Add(A=@n87, B=Multiply(A=NamedRerouteUsage(), B=@n83))
  NamedRerouteDeclaration(None=MaterialFunctionCall!("Xb1,PS4"=@n88, "PC-High&&Epic,XSX,PS5"=@n88, "PC-Perf,Android"=@n87, Switch=@n88, "PC-Low(NotPerf)&&Medium"=@n87) {materialFunction=/Game/Effects/Niagara/MasterMaterials/Functions/Common/PlatformSwitch_3D}.Result) {name=Ray Start Position, nodeColor=<rgba 1 0 0.9141 1>}
  n89 = Max(A=Subtract(A=NamedRerouteUsage(), B=NamedRerouteUsage())) {constB=0}
  n90 = Subtract(A=@n89, B=NamedRerouteUsage())
  n91 = NamedRerouteUsage()
  n92 = Floor(None=Divide(
    A=MaterialFunctionCall("Xb1,PS4"=@n90, "PC-High&&Epic,XSX,PS5"=@n90, "PC-Perf,Android"=@n89, Switch=@n90, "PC-Low(NotPerf)&&Medium"=@n89) {materialFunction=&PlatformSwitch_3D}.Result
    B=@n91
   ))
  NamedRerouteDeclaration(None=@n92) {name=Step Count, nodeColor=<rgba 0 0.3906 1 1>}
  n93 = Multiply(A=@n91, B=@n92)
  NamedRerouteDeclaration(None=Subtract(A=NamedRerouteUsage(), B=Add(A=@n86, B=@n93))) {name=Final Delta Distance, nodeColor=<rgba 1 0.9375 0 1>}
  NamedRerouteDeclaration(None=@n51.UVs) {name=DebugUV, nodeColor=<rgba 0 0.625 1 1>}
  NamedRerouteDeclaration(None=Multiply(A=@n40, B=@n41)) {name=Reactivity, nodeColor=<rgba 1 0 0.0234 1>}
  n94 = Power(Base=StaticSwitchParameter(
    True=MaterialFunctionCall() {materialFunction=&DebugTimeSine}.Result
    False=MaterialFunctionCall() {materialFunction=/Game/Packages/Fortress_SharedMaterials/Base_Material_Functions/BaseCharacterShaderFunctions/Audio/MF_AudioReactivity}.Amplitude
   ) {parameterName=AudioDebug}) {constExponent=3}
  n95 = Constant() {r=0.7}
  n96 = Add(
   A=Multiply(A=MaterialFunctionCall() {materialFunction=&MF_Hash}.Float) {constB=0.05}
   B=ConstantBiasScale(None=MaterialFunctionCall(Speed=@n95) {materialFunction=&DebugTimeSine}.Result) {bias=0.15, scale=0.3}
  )
  NamedRerouteDeclaration(None=Abs(None=LinearInterpolate(A=@n94, B=@n96, Alpha=@n65.FEAudio))) {name=AudioInput, nodeColor=<rgba 0.9453 1 0 1>}
  NamedRerouteDeclaration(None=@n43) {name=TODmask, nodeColor=<rgba 0.6406 0 1 1>}
  n97 = ScalarParameter() {defaultValue=1, parameterName=ScaleU2}
  n98 = ScalarParameter() {defaultValue=1.1, parameterName=Scalev2}
  n99 = Multiply(A=Multiply(A=AppendVector(A=@n97, B=@n98), B=@n17), B=@n19)
  n100 = ScalarParameter() {defaultValue=0.03, parameterName=SmallStarScaleMult}
  n101 = VectorParameter() {parameterName=SmallStarUVOffset, defaultValue=<rgba 0.45 0 0 1>}
  n102 = TextureSample(UVs=Add(A=Multiply(A=@n99, B=@n100) {constB=0.02}, B=AppendVector(A=@n101.R, B=@n101.R))) {
   desc=BaseSparkle
   samplerSource=SSM_Wrap_WorldGroupSettings
   samplerType=SAMPLERTYPE_LinearColor
   texture=/BRCosmetics/Accessories/FORT_Backpacks/Backpack_F_MED_DimeBlanketKnot/Texture/T_FallingSparklesTrail_M
  }
  n103 = Add(A=ScalarParameter() {defaultValue=0.77, parameterName=SparkleScreenSpaceUVOffset}, B=Divide(A=@n99, B=@n20))
  n104 = Constant()
  n105 = LinearInterpolate(A=Constant() {r=-0.1}, B=Constant() {r=-0.3}, Alpha=NamedRerouteUsage())
  n106 = TextureSample(UVs=Panner(Coordinate=@n103, Speed=AppendVector(A=@n104, B=@n105)) {speedY=-0.1}) {
   samplerSource=SSM_Wrap_WorldGroupSettings
   samplerType=SAMPLERTYPE_Masks
   texture=/BRCosmetics/Accessories/FORT_Backpacks/Backpack_F_MED_DimeBlanketKnot/Texture/T_DimeBlanketKnot_Stars
  }
  n107 = Multiply(A=@n106.B) {constB=5}
  n108 = Multiply(A=Constant() {r=1}, B=Add(A=@n63))
  NamedRerouteDeclaration(None=Multiply(
    A=MaterialFunctionCall(A=@n102.B, B=@n102.B, C=@n102.B, Alpha=@n106.B) {materialFunction=&Lerp_3Color}.Result
    B=Multiply(A=@n107, B=@n108)
   )) {name=SparklesIdle, nodeColor=<rgba 0 1 0.6953 1>}
  n109 = TextureSample(UVs=@n22) {
   samplerType=SAMPLERTYPE_Alpha
   texture=/BRCosmetics/Accessories/FORT_Backpacks/Backpack_F_MED_DimeBlanketKnot/Texture/T_HeartShapewithGradient
  }
  NamedRerouteDeclaration(None=Multiply(A=NamedRerouteUsage(), B=@n109.RGB)) {name="HeartShape(Day)", nodeColor=<rgba 0 0.5547 1 1>}
  n110 = Constant() {r=0.15}
  n111 = TextureSample(UVs=@n103) {
   samplerSource=SSM_Wrap_WorldGroupSettings
   samplerType=SAMPLERTYPE_Masks
   texture=&T_DimeBlanketKnot_Stars
  }
  n112 = Multiply(A=Time()) {constB=0.3}
  n113 = ScalarParameter() {defaultValue=1, parameterName=SparkleMaskOffset}
  n114 = OneMinus(None=Frac(None=Add(A=@n112, B=@n113)))
  n115 = Multiply(
   A=LinearInterpolate(
    A=MaterialFunctionCall(Speed=@n110) {materialFunction=&DebugTimeSine}.Result
    B=Multiply(A=@n108, B=@n64)
    Alpha=@n64
   )
   B=Multiply(A=@n111.R, B=SphereMask(A=@n111.G, B=@n114) {attenuationRadius=0.1, hardnessPercent=0})
  ) {constA=5}
  n116 = Multiply(
   A=@n115
   B=MaterialFunctionCall(
    Hue Shift Percentage=ConstantBiasScale(None=@n106.B) {bias=-0.5, scale=0.15}
    Texture=VectorParameter() {parameterName=StarReactivityColor, defaultValue=<rgba 0.7221 0 3 1>}.RGB
   ) {materialFunction=&HueShift}.Result
  )
  NamedRerouteDeclaration(None=Add(A=NamedRerouteUsage(), B=LinearInterpolate(A=@n115, B=@n116, Alpha=@n64))) {name="TwinkleSparkles(Night)", nodeColor=<rgba 1 0.1641 0 1>}
  NamedRerouteUsage()
  NamedRerouteUsage()
  NamedRerouteUsage()
  n117 = CollectionParameter() {collection=&FortniteMaterialParameters, parameterName=TimeOfDay_0-1}
  n118 = Constant() {r=0.53}
  OneMinus(None=Saturate(None=SphereMask(A=@n117, B=@n118) {attenuationRadius=0.38, hardnessPercent=99.8}))
  ScalarParameter() {defaultValue=-5000, parameterName=Param_2}
  ScalarParameter() {parameterName=FEAudio}
  SphereMask(A=NamedRerouteUsage(), B=Constant() {r=1}, Radius=ScalarParameter() {defaultValue=0.001, parameterName=ReactivityHeartDistance}) {attenuationRadius=0.01, hardnessPercent=0}
  ViewProperty() {property=MEVP_TemporalSampleIndex}
 ]
}
