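# Additive unlit low-end material for the DimeBlanketKnot backpack that draws a warm-tinted heart pulsing with audio.
# An AudioInput value blends cubed audio amplitude (or a debug sine via static switch) with a hashed sine pulse, selected by the FEAudio dynamic parameter.
# That value scales the heart texture's UVs about the center, and its sign gates the glow.
# A time-of-day mask is built from Fortnite's global parameter collection alongside a HeartShape(Day) fade.
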Material /BRCosmetics/Accessories/FORT_Backpacks/Backpack_F_MED_DimeBlanketKnot/Material/M_DimeBlanketKnot_Backpack_LowEndHeart
graph {
 properties {
  bUsedWithMeshDeformer=true
  bUsedWithNiagaraMeshParticles=true
  bUsedWithNiagaraSprites=true
  bUsedWithSkeletalMesh=true
  bUsedWithStaticLighting=true
  blendMode=BLEND_Additive
  pixelDepthOffsetMode=PDOM_Legacy
  shadingModel=MSM_Unlit
  usedShadingModels=Unlit
 }
 nodes [
  Constant() {r=1}
  n1 = CollectionParameter() {
   collection=&FortniteMaterialParameters
   parameterName=FortniteActiveTimeOfDay
  }
  NamedRerouteDeclaration(None=Power(
    Base=MaterialFunctionCall(Float4=@n1) {materialFunction=&BreakOutFloat4Components}.A
    Exp=ScalarParameter() {defaultValue=1, parameterName=TODPower}
   )) {name=TODmask, nodeColor=<rgba 0.6406 0 1 1>}
  n2 = Power(Base=StaticSwitchParameter(
    True=MaterialFunctionCall() {materialFunction=&DebugTimeSine}.Result
    False=MaterialFunctionCall() {materialFunction=/Game/Packages/Fortress_SharedMaterials/Base_Material_Functions/BaseCharacterShaderFunctions/Audio/MF_AudioReactivity}.Amplitude
   ) {parameterName=AudioDebug}) {constExponent=3}
  n3 = Constant() {r=0.7}
  n4 = Add(
   A=Multiply(A=MaterialFunctionCall() {materialFunction=&MF_Hash}.Float) {constB=0.05}
   B=ConstantBiasScale(None=MaterialFunctionCall(Speed=@n3) {materialFunction=&DebugTimeSine}.Result) {bias=0.15, scale=0.3}
  )
  n5 = DynamicParameter() {defaultValue=<rgba 0 0 0 0>, paramNames [FEAudio, ReactivitySwitch, Param3, Param4]}
  n6 = NamedRerouteDeclaration(None=Abs(None=LinearInterpolate(A=@n2, B=@n4, Alpha=@n5.FEAudio) {constAlpha=0})) {name=AudioInput, nodeColor=<rgba 0.9453 1 0 1>}
  n7 = MaterialFunctionCall(Texture Scale=Subtract(A=Constant() {r=1}, B=Multiply(A=@n6) {constB=0.2})) {materialFunction=&ScaleUVsByCenter}
  n8 = TextureSample(UVs=@n7.UVs) {
   samplerType=SAMPLERTYPE_Alpha
   texture=/BRCosmetics/Accessories/FORT_Backpacks/Backpack_F_MED_DimeBlanketKnot/Texture/T_HeartShapewithGradient
  }
  n9 = Multiply(A=Constant3Vector() {constant=<rgba 1 0.6548 0.147 1>}, B=Multiply(A=@n8.R, B=@n7.0-1 mask))
  n10 = Sign(None=@n6)
  NamedRerouteDeclaration(None=LinearInterpolate(A=LinearInterpolate(B=@n9, Alpha=@n10), Alpha=NamedRerouteUsage()) {constB=0}) {name="HeartShape(Day)", nodeColor=<rgba 0 0.5547 1 1>}
  NamedRerouteUsage()
 ]
}
