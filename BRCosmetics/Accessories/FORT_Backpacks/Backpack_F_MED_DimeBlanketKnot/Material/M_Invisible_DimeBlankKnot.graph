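# Invisible material for the DimeBlanketKnot backpack cosmetic, translucent and unlit, usable on skeletal, instanced and mesh-deformer meshes.
# A single zero constant drives both Emissive Color and Opacity through SetMaterialAttributes.
# Those attributes feed MF_CharacterParent_2023_PostFX with zeroed M and Fresnel inputs, keeping the shared character post-FX function in the chain.
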Material /BRCosmetics/Accessories/FORT_Backpacks/Backpack_F_MED_DimeBlanketKnot/Material/M_Invisible_DimeBlankKnot
graph {
 properties {
  bUseMaterialAttributes=true
  bUsedWithInstancedStaticMeshes=true
  bUsedWithMeshDeformer=true
  bUsedWithSkeletalMesh=true
  blendMode=BLEND_Translucent
  pixelDepthOffsetMode=PDOM_Legacy
  shadingModel=MSM_Unlit
  usedShadingModels=Unlit
 }
 nodes [
  n1 = Constant()
  MaterialFunctionCall(
   Material=SetMaterialAttributes!(MaterialAttributes, Emissive Color=@n1, Opacity=@n1) {attributeSetTypes [#1, #2]}
   M=Constant3Vector()
   Fresnel=Constant()
  ) {materialFunction=&MF_CharacterParent_2023_PostFX}
 ]
}
